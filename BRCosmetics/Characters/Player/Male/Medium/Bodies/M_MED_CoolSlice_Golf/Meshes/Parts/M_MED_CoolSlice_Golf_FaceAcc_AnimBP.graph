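# Animation Blueprint for the face accessory part of the CoolSlice Golf male medium character, driving it with poses copied from other meshes.
# The base pose copies from the attached parent, and a per-bone layered blend overlays the Body_Mesh pose on both upper-arm branches.
# OnCharacterCustomizationLoaded caches the custom base mesh as Body_Mesh and the Face part as Skel Mesh, which the AnimGraph does not read.
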AnimBlueprint /BRCosmetics/Characters/Player/Male/Medium/Bodies/M_MED_CoolSlice_Golf/Meshes/Parts/M_MED_CoolSlice_Golf_FaceAcc_AnimBP
target_skeleton=&/M_MED_CoolSlice_Golf_FaceAcc_Skeleton
graphs [
 AnimGraph {nodes [
   Copy Pose From Mesh() {bUseAttachedParent=true}
   n1 = Get Body_Mesh()
   Output Pose(Result=Layered blend per bone(
     BasePose=Copy Pose From Mesh() {bUseAttachedParent=true}
     BlendPoses_0=Copy Pose From Mesh(SourceMeshComponent=@n1) {bUseAttachedParent=true}
    ) {layerSetup [{branchFilters [{boneName=upperarm_l}, {boneName=upperarm_r}]}]}) {layerGroup=None}
  ]}
 EventGraph {nodes [
   n1 = Event OnCharacterCustomizationLoaded()
   > Set Body_Mesh(Body_Mesh=@n1.CustomBaseMesh)
   > Set Skel Mesh(Skel Mesh=Get Array Item with EFort Custom Part Type(Array=@n1.CustomMeshParts, PartType=Face))
   Event BlueprintUpdateAnimation()
  ]}
]
variables [
 Body_Mesh {type=SkeletalMeshComponent}
 Skel Mesh {type=SkeletalMeshComponent}
]
